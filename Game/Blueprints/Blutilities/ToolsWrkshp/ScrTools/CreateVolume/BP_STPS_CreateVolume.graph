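# Editor utility Blueprint that holds settings for a Create Volume tool in the ToolsWrkshp scripted tools.
# It exposes no logic: the event graph is empty, so it acts purely as a data container.
# Its variables define the volume type from an enum, a scale value, and the color and line thickness of a bounding box.
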EditorUtilityBlueprint /Game/Blueprints/Blutilities/ToolsWrkshp/ScrTools/CreateVolume/BP_STPS_CreateVolume
graphs [EventGraph {links [], nodes []}]
variables [
 Bounding Box Color {type=Color, default=<rgba 1 0 0.1569 1>}
 Bounding Box Line Thickness {type=number, default=5}
 Scale {type=number, default=100}
 Volume Type {type=E_CreateVolume_VolumeChoices, default=Volume}
]
