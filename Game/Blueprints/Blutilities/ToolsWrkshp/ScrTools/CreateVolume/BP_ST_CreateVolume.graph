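# Editor scriptable tool that spawns a chosen gameplay device or engine volume around the selected level actors.
# Clicks add traced actors to the selection, and selection changes redraw a line-set bounding box, ignoring created volumes.
# An action button drives CreateVolume, which derives size from selection bounds and a scale setting, applying it through named editor properties or brush-component scale within an undo transaction.
# Tool settings persist between sessions; Escape exits.
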EditorUtilityBlueprint /Game/Blueprints/Blutilities/ToolsWrkshp/ScrTools/CreateVolume/BP_ST_CreateVolume
graphs [
 CreateVolume {nodes [
   n1 = SpawnActorFromClass!(
    execute
    ActorClass:Actor Class Reference
    Location:Vector=Get BoundHeightOffsetToBottom() -> (BoundHeightOffsetToBottom:Vector)
    Rotation:Rotator
    bTransient:bool
   ) -> (then, ReturnValue:Actor Object Reference)
   > n2 = Set SpawnedActor(SpawnedActor:Actor Object Reference=@n1) -> (Output_Get:Actor Object Reference)
   n3 = Get SpawnedActor() -> (SpawnedActor:Actor Object Reference)
   SetSelectedLevelActors(
    execute=[@n17.then, @n41.then, @n2.then, @n44.then, @n19.then, @n35.then, @n15.then, @n39.then, @n23.then]
    self=Get EditorActorSubsystem()
    ActorsToSelect=Make Array("[0]":Actor Object Reference=@n3) -> (Array:Array of Actor Object References)
   )
   > n4 = Branch(Condition=Get RunSetProperties() -> (RunSetProperties:bool))
   n5 = Get Settings() -> (Settings:BP STPS Create Volume Object Reference)
   n6 = Get CurBoundExtent() -> (CurBoundExtent:Vector)
   n7 = vector * vector(
    A=vector / vector(A:Vector=vector * vector(A=@n6, B:Vector="2,2,2") -> (ReturnValue=~), B:Vector=Get DeviceVolBoxSize() -> (DeviceVolBoxSize:Vector)) -> (ReturnValue:Vector)
    B=SafeDivide(A=Get Scale!(self:BP STPS Create Volume Object Reference=@n5) -> (Scale:double), B=Make Literal Float(Value=100))
   ) -> (ReturnValue=~)
   n8 = BreakVector(InVec=@n7)
   > SetEditorProperty(
    Object=Get SpawnedActor()
    PropertyName=Get Prop Width Name() -> (Prop Width Name:Name=None)
    PropertyValue:double=@n8
    ChangeNotifyMode=Always
   )
   > SetEditorProperty(Object=Get SpawnedActor(), PropertyName=Get Prop Depth Name() -> (Prop Depth Name:Name=None), PropertyValue=@n8.Y, ChangeNotifyMode=Always)
   > n9 = SetEditorProperty(Object=Get SpawnedActor(), PropertyName=Get Prop Height Name() -> (Prop Height Name:Name=None), PropertyValue=@n8.Z, ChangeNotifyMode=Always)
   Add(
    execute=[@n4.else, @n9.then]
    TargetArray:Array of Actor Object References=Get AddedVolumes() -> (AddedVolumes:Array of Actor Object References)
    NewItem:"Actor Object Reference (by ref)"=Get SpawnedActor()
   ) {is=title K2Node_CallArrayFunction}
   > EndTransaction()
   CreateVolume() {is=title K2Node_FunctionEntry}
   > n10 = GetSelectedLevelActors(self=Get EditorActorSubsystem())
   > Branch(Condition="integer >= integer"(A=Length(TargetArray=@n10), B=1))
   > n11 = GetActorArrayBounds!(execute, Actors:Array of Actor Object References=@n10, bOnlyCollidingComponents:bool=false) -> (then, Center:Vector, BoxExtent:Vector)
   > Set CurBoundCenter(CurBoundCenter:Vector=@n11) -> (Output_Get:Vector)
   > Set CurBoundExtent(CurBoundExtent:Vector=@n11.BoxExtent) -> (Output_Get:Vector)
   > Set DeviceVolBoxSize(DeviceVolBoxSize:Vector="512,512,384") -> (Output_Get:Vector)
   n12 = Get CurBoundExtent()
   > Set BoundHeightOffsetToBottom(BoundHeightOffsetToBottom:Vector=vector - vector(A=Get CurBoundCenter() -> (CurBoundCenter:Vector), B=vector * vector(A=@n12, B="0,0,1") -> (ReturnValue=~))) -> (Output_Get:Vector)
   > BeginTransaction(Context=Get Undo Context() -> (Undo Context:String), Description=Get Undo Description() -> (Undo Description:Text))
   > n13 = Switch on E_CreateVolume_VolumeChoices!(
    execute
    Selection:E_CreateVolume_VolumeChoices Enum=Get Volume Type!(self:BP STPS Create Volume Object Reference=Get Settings()) -> (Volume Type:E_CreateVolume_VolumeChoices Enum=NewEnumerator8)
   ) -> (
    NewEnumerator8:Exec
    NewEnumerator7:Exec
    NewEnumerator9:Exec
    NewEnumerator1:Exec
    NewEnumerator10:Exec
    NewEnumerator11:Exec
    NewEnumerator13:Exec
    NewEnumerator14:Exec
    NewEnumerator16:Exec
    NewEnumerator17:Exec
    NewEnumerator18:Exec
    NewEnumerator19:Exec
   ) {is=title K2Node_SwitchEnum}
   n14 = SpawnActorFromClass(
    execute=@n13.NewEnumerator1
    ActorClass=/CRD_VolumetricRegion/BP_Device_DamageVolume_V2.BP_Device_DamageVolume_V2_C
    Location=Get BoundHeightOffsetToBottom()
   ) -> (ReturnValue:BP Device Damage Volume V2 Object Reference)
   > n15 = Set SpawnedActor(SpawnedActor=@n14)
   n16 = SpawnActorFromClass(
    execute=@n13.NewEnumerator11
    ActorClass=/CRD_VolumetricRegion/Device_MutatorVolume_V2.Device_MutatorVolume_V2_C
    Location=Get BoundHeightOffsetToBottom()
   ) -> (ReturnValue:Device Mutator Volume V2 Object Reference)
   > n17 = Set SpawnedActor(SpawnedActor=@n16)
   n18 = SpawnActorFromClass(
    execute=@n13.NewEnumerator13
    ActorClass=/CRD_VolumetricRegion/Device_SkydiveVolume_V2.Device_SkydiveVolume_V2_C
    Location=Get BoundHeightOffsetToBottom()
   ) -> (ReturnValue:Device Skydive Volume V2 Object Reference)
   > n19 = Set SpawnedActor(SpawnedActor=@n18)
   n20 = SpawnActorFromClass(execute=@n13.NewEnumerator16, ActorClass=/Script/Engine.BlockingVolume, Location=Get CurBoundCenter()) -> (ReturnValue:Blocking Volume Object Reference)
   > n21 = Set SpawnedActor(SpawnedActor=@n20)
   > Set DeviceVolBoxSize(DeviceVolBoxSize="200,200,200")
   > n22 = SetWorldScale3D!(
    execute
    self:Scene Component Object Reference=GetComponentByClass!(self:Actor Object Reference=@n21, ComponentClass:Actor Component Class Reference=/Script/Engine.BrushComponent) -> (ReturnValue:Brush Component Object Reference)
    NewScale:Vector=@n7
   ) -> (then)
   n23 = Set RunSetProperties(execute=[@n32.then, @n22.then, @n26.then, @n29.then], RunSetProperties:bool=false) -> (Output_Get:bool)
   n24 = SpawnActorFromClass(execute=@n13.NewEnumerator17, ActorClass=/Script/Engine.PostProcessVolume, Location=Get CurBoundCenter()) -> (ReturnValue:Post Process Volume Object Reference)
   > n25 = Set SpawnedActor(SpawnedActor=@n24)
   > Set DeviceVolBoxSize(DeviceVolBoxSize="200,200,200")
   > n26 = SetWorldScale3D(self=GetComponentByClass(self=@n25, ComponentClass=&BrushComponent), NewScale=@n7)
   n27 = SpawnActorFromClass(execute=@n13.NewEnumerator18, ActorClass=/Script/FortniteGame.FortUndergroundVolume, Location=Get CurBoundCenter()) -> (ReturnValue:Fort Underground Volume Object Reference)
   > n28 = Set SpawnedActor(SpawnedActor=@n27)
   > Set DeviceVolBoxSize(DeviceVolBoxSize="200,200,200")
   > n29 = SetWorldScale3D(self=GetComponentByClass(self=@n28, ComponentClass=&BrushComponent), NewScale=@n7)
   n30 = SpawnActorFromClass(execute=@n13.NewEnumerator19, ActorClass=/Script/FortniteGame.FortWaterBodyExclusionVolume, Location=Get CurBoundCenter()) -> (ReturnValue:Fort Water Body Exclusion Volume Object Reference)
   > n31 = Set SpawnedActor(SpawnedActor=@n30)
   > Set DeviceVolBoxSize(DeviceVolBoxSize="200,200,200")
   > n32 = SetWorldScale3D(self=GetComponentByClass(self=@n31, ComponentClass=&BrushComponent), NewScale=@n7)
   n33 = SpawnActorFromClass(
    execute=@n13.NewEnumerator7
    ActorClass=/CRD_VolumetricRegion/Device_Barrier_V2_Placed.Device_Barrier_V2_Placed_C
    Location=Get BoundHeightOffsetToBottom()
   ) -> (ReturnValue:Device Barrier V2 Placed Object Reference)
   > n34 = Set SpawnedActor(SpawnedActor=@n33)
   Set Prop Width Name(execute=[@n37.then, @n34.then], Prop Width Name:Name=BarrierWidth) -> (Output_Get:Name=None)
   > Set Prop Height Name(Prop Height Name:Name=BarrierHeight) -> (Output_Get:Name=None)
   > n35 = Set Prop Depth Name(Prop Depth Name:Name=BarrierDepth) -> (Output_Get:Name=None)
   n36 = SpawnActorFromClass(
    execute=@n13.NewEnumerator8
    ActorClass=/CRD_VolumetricRegion/Device_AI_Navigation_Modification.Device_AI_Navigation_Modification_C
    Location=Get BoundHeightOffsetToBottom()
   ) -> (ReturnValue:Device AI Navigation Modification Object Reference)
   > n37 = Set SpawnedActor(SpawnedActor=@n36)
   n38 = SpawnActorFromClass(execute=@n13.NewEnumerator9, Location=Get BoundHeightOffsetToBottom()) -> (ReturnValue:Actor Object Reference)
   > n39 = Set SpawnedActor(SpawnedActor=@n38)
   n40 = SpawnActorFromClass(
    execute=@n13.NewEnumerator14
    ActorClass=/CRD_Volume/Device_Volume.Device_Volume_C
    Location=vector + vector(A=Get BoundHeightOffsetToBottom(), B="0,0,192")
   ) -> (ReturnValue:Device Volume Object Reference)
   > Set SpawnedActor(SpawnedActor=@n40)
   > Set Prop Width Name(Prop Width Name=VolumeWidth) -> (Output_Get=None)
   > Set Prop Height Name(Prop Height Name=VolumeHeight) -> (Output_Get=None)
   > n41 = Set Prop Depth Name(Prop Depth Name=VolumeDepth) -> (Output_Get=None)
   n42 = Get BoundHeightOffsetToBottom()
   n43 = SpawnActorFromClass(execute=@n13.NewEnumerator10, ActorClass=/CRD_VolumetricRegion/Device_Fire_Volume_V2.Device_Fire_Volume_V2_C, Location=@n42) -> (ReturnValue:Device Fire Volume V2 Object Reference)
   > n44 = Set SpawnedActor(SpawnedActor=@n43)
   n45 = GetLocalBounds!(self:Static Mesh Component Object Reference=GetComponentByClass(self=@n3, ComponentClass=/CRD_Volume/FNE_VolumeOverlapComponent.FNE_VolumeOverlapComponent_C) -> (ReturnValue:FNE Volume Overlap Component Object Reference)) -> (Min:Vector, Max:Vector)
   n46 = MakeBox!(Min:Vector=@n45, Max:Vector=@n45.Max) -> (ReturnValue:Box Structure)
   GetBoxCenterSize!(Box:Box Structure=@n46) -> (Center:Vector, Dimensions:Vector)
   GetExpandedBox!(Box:Box Structure=@n46, ExpandBy:Vector="0, 0, 0") -> (ReturnValue:Box Structure)
   "Get Size (Box)"!(InBox:"Box Structure (by ref)"=@n46) -> (ReturnValue:Vector) {is=title K2Node_CallFunction}
   vector + vector(A=@n42, B="0,0,-16")
   Get SpawnedActor()
  ]}
 EventGraph {nodes [
   Event OnScriptSetup()
   > n1 = Sequence!(execute) -> (then_0, then_1, then_2:Exec, then_3:Exec, then_4:Exec, then_5:Exec, then_6:Exec, then_7:Exec, then_8:Exec)
   > K2Node_Composite_4 = ConfigPropertySets!(execute) -> (then:Exec) {is=title K2Node_Composite}
   K2Node_Composite_0 = Add Behaviours(execute=@n1.then_1) {is=title K2Node_Composite}
   K2Node_Composite_1 = Add Header Widget(execute=@n1.then_4) {is=title K2Node_Composite}
   K2Node_Composite_6 = BindToSelChanged(execute=@n1.then_5) {is=title K2Node_Composite}
   K2Node_Composite_7 = Add Line Set(execute=@n1.then_6) {is=title K2Node_Composite}
   K2Node_Composite_8 = Init BBox(execute=@n1.then_7) {is=title K2Node_Composite}
   K2Node_Composite_9 = Property Watches(execute=@n1.then_8) {is=title K2Node_Composite}
   n2 = Event OnScriptTick()
   > Set CurTime(CurTime=float + float(A=@n2.DeltaTime, B=Get CurTime()))
   > n3 = Sequence!(execute) -> (then_0, then_1, then_2)
   n4 = TimedClearUserWarning!(
    execute=@n3.then_1
    self
    ClearAfterTime:double=Get Clear After Time() -> (Clear After Time:double)
    CurTime:double=Get CurTime()
    LastClearTime:double=Get Last Clear Time() -> (Last Clear Time:double)
   ) -> (then, NewLastClearTime:double)
   > Set Last Clear Time(Last Clear Time:double=@n4) -> (Output_Get:double)
   n5 = Branch(execute=@n3.then_2, Condition=Get FirstInitRun())
   Set FirstInitRun(execute=@n5.else)
   > n6 = Sequence!(execute) -> (then_0, then_1)
   n7 = AddExpandableBox(execute=@n6.then_1, self=Get Header Widget())
   > n8 = Set Actions Box(Actions Box=@n7)
   > SetUpBox(self=@n8, Name=Get Actions Box Label(self=Get Header Widget()))
   > n9 = Construct Object from Class(Class=/Game/Blueprints/Blutilities/ToolsWrkshp/ScrTools/CreateVolume/SupportFiles/BP_UW_ActionsButton_CreateVolume.BP_UW_ActionsButton_CreateVolume_C) -> (ReturnValue:BP UW Actions Button Create Volume Object Reference)
   > Set ActionsButton(ActionsButton:BP UW Actions Button Create Volume Object Reference=@n9) -> (Output_Get:BP UW Actions Button Create Volume Object Reference)
   > AddChildToVerticalBox(
    self=Get DynamicEntries(self=Get Actions Box())
    Content=Get ActionsButton() -> (ActionsButton:BP UW Actions Button Create Volume Object Reference)
   )
   > Bind Event to Event Create Volume Clicked(self:BP UW Actions Button Create Volume Object Reference=Get ActionsButton(), Delegate=Create Event()) {is=title K2Node_AddDelegate}
   Event OnScriptShutdown()
   > K2Node_Composite_5 = SavePropSetStates() {is=title K2Node_Composite}
   Reroute Node(InputPin:Wildcard) -> (OutputPin:Wildcard)
  ]}
 Add Behaviours {
  path=EventGraph.K2Node_Composite_0.Add Behaviours
  nodes [
   AddSingleClickBehavior!(
    execute
    self:Scriptable Modular Behavior Tool Object Reference
    IfHitByClick:Delegate=Create Event()
    OnHitByClick:Delegate=Create Event()
    CaptureCheck:Delegate
    CapturePriority:int=100
    MouseButton:EScriptableToolMouseButton Enum=LeftButton
    bHitTestOnRelease:bool=true
   ) -> (then)
   > AddSingleKeyInputBehavior!(
    execute
    self:Scriptable Modular Behavior Tool Object Reference
    OnKeyPressed:Delegate=Create Event()
    OnKeyReleased:Delegate
    OnForceEndCaptureFuncIn:Delegate
    Key:Key Structure=Escape
    CaptureCheck:Delegate
    CapturePriority:int=1000000
   ) -> (then)
   Inputs()
   Outputs()
  ]
 }
 Add Header Widget {
  path=EventGraph.K2Node_Composite_1.Add Header Widget
  nodes [
   Inputs()
   > n1 = Sequence()
   > n2 = Construct Object from Class(Class=&BP_UW_SharedHeaderWidget_C) -> (ReturnValue:BP UW Shared Header Widget Object Reference)
   > n3 = Set Header Widget(Header Widget:BP UW Shared Header Widget Object Reference=@n2) -> (Output_Get:BP UW Shared Header Widget Object Reference)
   > SetToolPanelToolHeaderWidget(Widget=@n3)
   DisplayUserWarningMessage(execute=@n1.then_1, Message="LOCTEXT(\" \")")
   Outputs()
  ]
 }
 ConfigPropertySets {
  path=EventGraph.K2Node_Composite_4.ConfigPropertySets
  nodes [
   Inputs()
   > n1 = Sequence!(execute) -> (then_0, then_1, then_2:Exec, then_3:Exec)
   n2 = AddPropertySetOfType(
    execute=@n1.then_1
    PropertySetType=/Game/Blueprints/Blutilities/ToolsWrkshp/ScrTools/Shared/SharedPropertySets/BP_STPS_SharedWidgetState.BP_STPS_SharedWidgetState_C
    Identifier=Get SharedStateIdentifier() -> (SharedStateIdentifier:String)
   )
   > n3 = Cast To BP_STPS_SharedWidgetState!(execute, Object=@n2) -> (then, CastFailed, AsBP STPS Shared Widget State:BP STPS Shared Widget State Object Reference) {is=title K2Node_DynamicCast}
   > Set SharedState Settings(SharedState Settings:BP STPS Shared Widget State Object Reference=@n3) -> (Output_Get:BP STPS Shared Widget State Object Reference)
   > SetPropertySetVisibleByName!(execute, self:Scriptable Interactive Tool Object Reference, Identifier:String=Get SharedStateIdentifier(), bVisible:bool=false) -> (then)
   > RestorePropertySetSettings(
    PropertySet=Get SharedState Settings() -> (SharedState Settings:BP STPS Shared Widget State Object Reference)
    SaveKey=Get SaveSettings_Identifier() -> (SaveSettings_Identifier:String)
   )
   Outputs!(then:Exec=@n1.then_3)
   n4 = AddPropertySetOfType(
    execute=@n1.then_0
    PropertySetType=&/BP_STPS_CreateVolume.BP_STPS_CreateVolume_C
    Identifier=Get Settings_Identifier() -> (Settings_Identifier:String)
   )
   > n5 = Cast To BP_STPS_CreateVolume!(execute, Object=@n4) -> (then, CastFailed, AsBP STPS Create Volume:BP STPS Create Volume Object Reference) {is=title K2Node_DynamicCast}
   > n6 = Set Settings(Settings:BP STPS Create Volume Object Reference=@n5) -> (Output_Get:BP STPS Create Volume Object Reference)
   > RestorePropertySetSettings(PropertySet=@n6, SaveKey=Get SaveSettings_Identifier())
  ]
 }
 SavePropSetStates {
  path=EventGraph.K2Node_Composite_5.SavePropSetStates
  nodes [
   Inputs()
   > SavePropertySetSettings(PropertySet=Get Settings(), SaveKey=Get SaveSettings_Identifier())
   n1 = Get Shortcut Box() -> (Shortcut Box:BP UW Expandable Box Object Reference)
   > Set ShortcutsOpenOnExit!(
    execute
    ShortcutsOpenOnExit:bool=IsChecked(self=Get ExpandButton!(self:BP UW Expandable Box Object Reference=@n1) -> (ExpandButton:Check Box Object Reference))
    self:BP STPS Shared Widget State Object Reference=Get SharedState Settings()
   ) -> (then, Output_Get:bool)
   > SavePropertySetSettings(PropertySet=Get SharedState Settings(), SaveKey=Get SaveSettings_Identifier())
   Outputs!()
  ]
 }
 BindToSelChanged {
  path=EventGraph.K2Node_Composite_6.BindToSelChanged
  nodes [
   Inputs()
   > n1 = GetSelectionSet(self=Get LevelEditorSubsystem())
   > Bind Event to On Selection Change(self:Typed Element Selection Set Object Reference=@n1, Delegate=Create Event()) {is=title K2Node_AddDelegate}
   Outputs()
  ]
 }
 Add Line Set {
  path=EventGraph.K2Node_Composite_7.Add Line Set
  nodes [
   Inputs()
   > n1 = AddLineSet!(execute, self:Scriptable Interactive Tool Object Reference) -> (then, ReturnValue:Scriptable Tool Line Set Object Reference)
   > Set Line Set(Line Set:Scriptable Tool Line Set Object Reference=@n1) -> (Output_Get:Scriptable Tool Line Set Object Reference)
   > For Loop(LastIndex=11)
   > n2 = AddLine!(execute, self:Scriptable Tool Line Set Object Reference=Get Line Set() -> (Line Set:Scriptable Tool Line Set Object Reference)) -> (then, ReturnValue:Scriptable Tool Line Object Reference)
   > Add(
    TargetArray:Array of Scriptable Tool Line Object References=Get BBox Lines() -> (BBox Lines:Array of Scriptable Tool Line Object References)
    NewItem:"Scriptable Tool Line Object Reference (by ref)"=@n2
   )
   Outputs()
  ]
 }
 Init BBox {
  path=EventGraph.K2Node_Composite_8.Init BBox
  nodes [
   Inputs()
   > n1 = GetSelectedLevelActors(self=Get EditorActorSubsystem())
   > n2 = Set ActorsSelectedOnStartup(ActorsSelectedOnStartup:Array of Actor Object References=@n1) -> (Output_Get:Array of Actor Object References)
   > UpdateBBox!(execute, self, ActorsArray:Array of Actor Object References=@n2) -> (then)
   Outputs()
  ]
 }
 Property Watches {
  path=EventGraph.K2Node_Composite_9.Property Watches
  nodes [
   Inputs()
   > n1 = WatchProperty(PropertySet=Get Settings(), PropertyName=Bounding Box Line Thickness, OnModified=Create Event())
   > WatchProperty(PropertySet=@n1, PropertyName=Bounding Box Color, OnModified=Create Event())
   Outputs()
  ]
 }
 IfHitByClick {nodes [
   n1 = IfHitByClick!() -> (
    then
    ClickPos_WorldRay:Ray Structure
    ClickPos_bHas2D:bool=~
    ClickPos_ScreenPosition:Vector 2D Structure
    MouseButton:EScriptableToolMouseButton Enum=LeftButton
   ) {is=title K2Node_FunctionEntry}
   n2 = GetRayStartEnd!(Ray:Ray Structure=@n1, StartDistance:double=0, EndDistance:double=0) -> (StartPoint:Vector, EndPoint:Vector)
   > n3 = Line Trace By Channel!(
    execute
    Start:Vector=@n2
    End:Vector=@n2.EndPoint
    TraceChannel:ETraceTypeQuery Enum
    bTraceComplex:bool=true
    ActorsToIgnore:Array of Actor Object References
    DrawDebugType:EDrawDebugTrace Enum
    bIgnoreSelf:bool
    TraceColor:Linear Color Structure
    TraceHitColor:Linear Color Structure
    DrawTime:float
   ) -> (
    then
    OutHit_bBlockingHit:bool
    OutHit_bInitialOverlap:bool
    OutHit_Time:float
    OutHit_Distance:float
    OutHit_Location:Vector
    OutHit_ImpactPoint:Vector
    OutHit_Normal:Vector
    OutHit_ImpactNormal:Vector
    OutHit_PhysMat:Physical Material Object Reference
    OutHit_HitActor:Actor Object Reference
    OutHit_HitComponent:Primitive Component Object Reference
    OutHit_HitBoneName:Name=None
    OutHit_BoneName:Name=None
    OutHit_HitItem:int=0
    OutHit_ElementIndex:int=0
    OutHit_FaceIndex:int=0
    OutHit_TraceStart:Vector
    OutHit_TraceEnd:Vector
    ReturnValue:bool
   )
   > Set HitLoc(HitLoc:Vector=@n3.OutHit_Location) -> (Output_Get:Vector)
   > Set Out Hit Hit Actor(Out Hit Hit Actor:Actor Object Reference=@n3.OutHit_HitActor) -> (Output_Get:Actor Object Reference)
   > Return Node!(
    execute
    ReturnValue_bHit:bool=@n3.ReturnValue
    ReturnValue_HitDepth:double=340282346638528859811704183484516925440
    ReturnValue_HitNormal:Vector="0,0,1"
    ReturnValue_bHasHitNormal:bool=False
    ReturnValue_HitIdentifier:int=0
    ReturnValue_HitObject:Object Reference
   )
  ]}
 On BBox Appearance Changed {nodes [
   On BBox Appearance Changed!() -> (then, PropertySet:Scriptable Interactive Tool Property Set Object Reference, PropertyName:String) {is=title K2Node_FunctionEntry}
   > OnSelectionChanged!(execute, self, SelectionSet:Typed Element Selection Set Object Reference) -> (then)
  ]}
 On Esc Pressed {nodes [
   On Esc Pressed!() -> (then, Key:Key Structure=None, Modifiers:Scriptable Tool Modifier States Structure) {is=title K2Node_FunctionEntry}
   > RequestToolShutdown!(execute, self:Scriptable Interactive Tool Object Reference, bAccept:bool=true, bShowUserPopupMessage:bool=false, UserMessage:Text) -> (then)
  ]}
 OnClearAllUtilityActors {nodes [
   OnClearAllUtilityActors() {is=title K2Node_FunctionEntry}
   > n1 = GetAllActorsWithTag!(execute, Tag:Name=Get ScaleActorTagName() -> (ScaleActorTagName:Name=None)) -> (then, OutActors:Array of Actor Object References)
   n2 = Length(TargetArray=@n1)
   > n3 = Branch(Condition="integer > integer"(A=@n2))
   n4 = "To Text (String)"(InString=Get ClearTransientActorsWarnIniName() -> (ClearTransientActorsWarnIniName:String))
   n5 = Get ToolName(None=Self-Reference()) -> (ToolName:Text)
   n6 = "Not Equal (Integer)"(A=@n2, B=1)
   n7 = Get ClearTransientActorsWarnMessage_p() -> (ClearTransientActorsWarnMessage_p:Text)
   n8 = Get ClearTransientActorsWarnMessage_s() -> (ClearTransientActorsWarnMessage_s:Text)
   > n9 = Show Suppressable Warning Dialog!(
    execute
    Title:"Text (by ref)"=Get ClearTransientActorsWarnTitle() -> (ClearTransientActorsWarnTitle:Text)
    Message:"Text (by ref)"=Format Text!(Format=SelectText(A=@n7, B=@n8, bPickA=@n6), 0:int=@n2) -> (Result)
    InIniSettingName:String="To String (Text)"(InText=Format Text(Format=@n4, 0:Text=@n5))
    InIniSettingFileNameOverride:String
    bDefaultValue:bool=true
   ) -> (then, ReturnValue:bool)
   > Branch(Condition=@n9)
   > BeginTransaction(Context=Get TransactionContext() -> (TransactionContext:String), Description="LOCTEXT(\"Delete Actor\")")
   > n10 = For Each Loop(Array:Array of Actor Object References=@n1) -> (Array Element:Actor Object Reference)
   EndTransaction(execute=@n10.Completed)
   Destroy Actor(execute=@n10.LoopBody, self=@n10)
   DisplayUserWarningMessage(execute=@n3.else, Message=Get ClearTransientNoneFoundWarn() -> (ClearTransientNoneFoundWarn:Text))
   > Set Last Clear Time(Last Clear Time=Get CurTime())
   Get ClearTransientActorsWarnIniName()
  ]}
 OnHitByClick {nodes [
   OnHitByClick!() -> (
    then
    ClickPos:Input Device Ray Structure
    Modifiers:Scriptable Tool Modifier States Structure
    MouseButton:EScriptableToolMouseButton Enum=LeftButton
   ) {is=title K2Node_FunctionEntry}
   > n1 = GetSelectedLevelActors(self=Get EditorActorSubsystem())
   > n2 = Set CurSelActors(CurSelActors:Array of Actor Object References=@n1) -> (Output_Get:Array of Actor Object References)
   > Add Unique(
    TargetArray:Array of Actor Object References=@n2
    NewItem:"Actor Object Reference (by ref)"=Get Out Hit Hit Actor() -> (Out Hit Hit Actor:Actor Object Reference)
   )
   > n3 = Branch(Condition=Identical!(
     ArrayA:Array of Actor Object References=Get CurSelActors() -> (CurSelActors:Array of Actor Object References)
     ArrayB:Array of Actor Object References=@n1
    ) -> (ReturnValue:bool) {is=title K2Node_CallArrayFunction})
   SetSelectedLevelActors(execute=@n3.else, self=Get EditorActorSubsystem(), ActorsToSelect=Get CurSelActors())
  ]}
 OnSelectionChanged {nodes [
   OnSelectionChanged!() -> (then, SelectionSet:Typed Element Selection Set Object Reference) {is=title K2Node_FunctionEntry}
   > n1 = GetSelectedLevelActors(self=Get EditorActorSubsystem())
   > Set Filtered Sel Level Actors(Filtered Sel Level Actors:Array of Actor Object References=@n1) -> (Output_Get:Array of Actor Object References)
   > n2 = For Each Loop(Array=Get AddedVolumes())
   UpdateBBox(execute=@n2.Completed, ActorsArray=Get Filtered Sel Level Actors() -> (Filtered Sel Level Actors:Array of Actor Object References))
   Remove Item!(
    execute=@n2.LoopBody
    TargetArray:Array of Actor Object References=Get Filtered Sel Level Actors()
    Item:"Actor Object Reference (by ref)"=@n2
   ) -> (then, ReturnValue:bool)
  ]}
 PlaceActorBaseOnGround {nodes [
   n1 = PlaceActorBaseOnGround!() -> (then, Actor:Actor Object Reference) {is=title K2Node_FunctionEntry}
   n2 = Cast To StaticMeshActor!(execute, Object=@n1) -> (then, CastFailed, AsStatic Mesh Actor:Static Mesh Actor Object Reference) {is=title K2Node_DynamicCast}
   n3 = GetComponentByClass(self=@n1, ComponentClass=&StaticMeshComponent) -> (ReturnValue:Static Mesh Component Object Reference)
   n4 = GetBounds!(self:Static Mesh Object Reference=Get StaticMesh(self=@n3)) -> (ReturnValue_Origin:Vector, ReturnValue_BoxExtent:Vector, ReturnValue_SphereRadius:float)
   n5 = Set Delta Z(execute=[@n2.then, @n9.Is Valid], Delta Z:double=float - float(A=BreakVector(InVec=@n4.ReturnValue_BoxExtent).Z, B=BreakVector(InVec=@n4).Z)) -> (Output_Get:double)
   Add Actor World Offset!(
    execute=[@n8.then, @n5.then]
    self:Actor Object Reference=@n1
    DeltaLocation_X:double=0
    DeltaLocation_Y:double=0
    DeltaLocation_Z:double=Get Delta Z() -> (Delta Z:double)
    bSweep:bool=false
    bTeleport:bool=false
   ) -> (then, SweepHitResult:Hit Result Structure)
   n6 = GetComponentByClass(self=@n1, ComponentClass=/Script/Engine.SkeletalMeshComponent) -> (ReturnValue:Skeletal Mesh Component Object Reference)
   Is Valid(exec=[@n2.CastFailed, @n9.Is Not Valid], InputObject=@n6)
   n7 = GetBounds(self:Skeletal Mesh Object Reference=GetSkeletalMeshAsset!(self:Skeletal Mesh Component Object Reference=@n6) -> (ReturnValue:Skeletal Mesh Object Reference))
   > n8 = Set Delta Z(Delta Z=float - float(A=BreakVector(InVec=@n7.ReturnValue_BoxExtent).Z, B=BreakVector(InVec=@n7).Z))
   Get StaticMeshComponent(self=@n2)
   n9 = Is Valid(exec=@n1.then, InputObject=@n3)
  ]}
 SetCommonActorProperties {nodes [
   n1 = Inputs!() -> (execute, actor:Actor Object Reference)
   > SetActorLabel(self=@n1, NewActorLabel=Get ActorFinalName() -> (ActorFinalName:String), bMarkDirty=false)
   > SetEditorProperty(Object=@n1, PropertyName=bIsEditorOnlyActor, PropertyValue:bool=Get Editor Only() -> (Editor Only:bool))
   > Set Tags!(
    execute
    Tags:Array of Names=Make Array("[0]":Name=Get ScaleActorTagName() -> (ScaleActorTagName=None)) -> (Array:Array of Names)
    self:Actor Object Reference=@n1
   ) -> (then, Output_Get:Array of Names)
   > SetSelectedLevelActors(self=Get EditorActorSubsystem(), ActorsToSelect=Make Array("[0]":Actor Object Reference=@n1) -> (Array:Array of Actor Object References))
   > Outputs!(then:Exec)
  ]}
 TimedClearUserWarning {nodes [
   n1 = TimedClearUserWarning!() -> (then, ClearAfterTime:double=3, CurTime:double=~, LastClearTime:double=~) {is=title K2Node_FunctionEntry}
   > Set LastClearTimeUpdated(LastClearTimeUpdated:double=@n1.LastClearTime) -> (Output_Get:double)
   > Branch(Condition="float > float"(A=@n1.CurTime, B=float + float(A=@n1, B=@n1.LastClearTime)))
   > Branch()
   > Set LastClearTimeUpdated(LastClearTimeUpdated=@n1.CurTime)
   > ClearUserMessages!(execute, self:Scriptable Interactive Tool Object Reference, bHelpMessage:bool=true, bWarningMessage:bool=true) -> (then)
   > Return Node!(execute, NewLastClearTime:double=Get LastClearTimeUpdated() -> (LastClearTimeUpdated:double))
  ]}
 UpdateBBox {nodes [
   n1 = UpdateBBox!() -> (then, ActorsArray:Array of Actor Object References) {is=title K2Node_FunctionEntry}
   > n2 = Branch(Condition="integer > integer"(A=Length(TargetArray=@n1)))
   > n3 = GetActorArrayBounds(Actors=Get ActorsArray() -> (ActorsArray:Array of Actor Object References), bOnlyCollidingComponents=false)
   n4 = Get Settings()
   > DrawBBox!(
    execute
    BoundOrigin:Vector=@n3
    BoundExtent:Vector=@n3.BoxExtent
    Box Lines:Array of Scriptable Tool Line Object References=Get BBox Lines()
    Color:Color Structure=Get Bounding Box Color!(self:BP STPS Create Volume Object Reference=@n4) -> (Bounding Box Color:Color Structure)
    Thickness:float=Get Bounding Box Line Thickness!(self:BP STPS Create Volume Object Reference=@n4) -> (Bounding Box Line Thickness:double)
    DepthBias:float=0
   ) -> (then)
   DrawBBox(execute=@n2.else, BoundOrigin="0, 0, 0", BoundExtent="0, 0, 0", Box Lines=Get BBox Lines(), Thickness=5, DepthBias=0)
  ]}
]
variables [
 Actions Box {type=BP_UW_ExpandableBox_C}
 ActionsButton {type=BP_UW_ActionsButton_CreateVolume_C, default=null}
 ActorFinalName {type=string, default=""}
 ActorsSelectedOnStartup {type="array<Actor>"}
 AddedSkelMeshActor {type=SkeletalMeshActor, default=null}
 AddedVolumes {type="array<Actor>"}
 AllAddedScaleActors {type="array<Actor>"}
 BBox Lines {type="array<ScriptableToolLine>", default []}
 Clear After Time {type=number, default=7}
 ClearTransientActorsWarnIniName {type=string, default="ToolsWorkshop_{0}_ClearAllWarning"}
 ClearTransientActorsWarnMessage_p {type=string, default="This will remove {0} Helper Assets in the level. Delete anyway?"}
 ClearTransientActorsWarnMessage_s {type=string, default=This will remove 1 Helper Asset in the level. Delete anyway?}
 ClearTransientActorsWarnTitle {type=string, default=Clear Helper Assets}
 ClearTransientNoneFoundWarn {type=string, default=No objects found to remove.}
 ColorCalibMesh {type=StaticMesh, default=/Engine/EditorMeshes/ColorCalibrator/SM_ColorCalibrator}
 CurTime {type=number}
 DeviceVolBoxSize {type=Vector, default=<512 512 384>}
 Editor Only {type=boolean, default=true}
 EditorOnlyTools {type=BP_TravelTime_EditorOnlyUtils_C, default=null}
 FirstInitRun {type=boolean}
 HaveAlreadyWarnedGameView {type=boolean}
 Header Widget {type=BP_UW_SharedHeaderWidget_C}
 HitLoc {type=Vector}
 Is Game View On Tool Start {type=boolean}
 Last Clear Time {type=number}
 Line Set {type=ScriptableToolLineSet, default=null}
 Mannequin_SKM {type=SkeletalMesh, default=/Game/Creative/Devices/Mannequin/Meshes/CP_Device_Mannequin}
 Out Hit Hit Actor {type=Actor, default=null}
 SaveSettings_Identifier {type=string, default=SaveSettings}
 ScaleActorTagName {type=string, default=ScaleActor}
 Settings {type=BP_STPS_CreateVolume_C, default=null}
 Settings_Identifier {type=string, default=Settings}
 SharedState Settings {type=BP_STPS_SharedWidgetState_C, default=null}
 SharedStateIdentifier {type=string, default=SharedStateSettings}
 Shortcut Box {type=BP_UW_ExpandableBox_C}
 TransactionContext {type=string, default=""}
 Transient {type=boolean, default=true}
 Undo Context {type=string, default=Create Volume}
 Undo Description {type=string, default=Add Actors}
]
